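# Unlit masked material for Niagara sprite flower effects in the Sleek Glasses Jacket character's FX.
# One packed texture drives it: red indexes a petal color curve atlas row, inverted blue darkens the color, and green is routed separately.
# A static switch can substitute particle color before the shared emissive adjustment function.
# Texture mip bias rises with camera-to-particle distance, clamped, so distant sprites sample blurrier mips.
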Material /Game/Characters/Player/Male/Medium/Bodies/M_MED_Sleek_Glasses_Jacket/FX/M_SleekFlowers
graph {
 properties {bUsedWithNiagaraSprites=true, blendMode=BLEND_Masked, pixelDepthOffsetMode=PDOM_Legacy, shadingModel=MSM_Unlit, usedShadingModels=Unlit}
 nodes [
  n1 = ParticleColor()
  n2 = Divide(A=Distance(A=CameraPositionWS(), B=ParticlePositionWS()), B=Constant() {r=1024})
  n3 = TextureSampleParameter2D!(UVs=TextureCoordinate(), Bias=LinearInterpolate(A=Constant(), B=Constant() {r=1}, Alpha=Clamp(None=@n2)), Apply View MipBias) {
   mipValueMode=TMVM_MipBias
   parameterName=GreyScaleInput
   samplerType=SAMPLERTYPE_LinearColor
   texture=&/T_FlowersExpansion_M
  }
  n4 = CurveAtlasRowParameter(CurveTime=@n3.R) {
   atlas=&/CA_SleekFlowerColors
   curve=&/C_SleekPetalColor
  }
  MaterialFunctionCall(Emissive=Multiply(A=StaticSwitchParameter(True=@n1.RGB, False=@n4) {parameterName=UseParticleColor}, B=OneMinus(None=@n3.B))) {materialFunction=&MF_FX_EmissiveAdjustments}
  Reroute(None=@n3.G)
 ]
}
